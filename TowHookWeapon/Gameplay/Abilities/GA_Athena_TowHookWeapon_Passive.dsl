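; Passive gameplay ability for the tow hook weapon that applies a blocking effect to its owner while the hook is active.
; Event-triggered activation switches on the tag: ProjSpawned applies the active effect, while Stop removes effects by their granted tag.
; ProjSpawned also caches the Fort player pawn and, if the avatar is valid, starts a failsafe timer that force-ends the ability.
; Ending the ability removes the tagged effects again as cleanup.
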(event Ability|EventActivateAbilityFromEvent
  (params (EventData :type Gameplay%20Event%20Data%20Structure%20%28by%20ref%29))
  (bind _returnvalue (Ability|GetAvatarActorfromActorInfo))
  ;[1 1 1] Apply the blocking effects while the ability is active. Remove the effect when the ability is called to stop
  (|CallParentFunction EventData)
  (switch Utilities|FlowControl|Switch|SwitchonGameplayTag (Utilities|Struct|BreakGameplayEventData EventData)
    (:Event.Athena.Item.TowHookWeapon.Stop
      (Ability|RemoveGameplayEffectFromOwnerWithGrantedTags (Variables|Default|GetT_Active)))
    (:Event.Athena.Item.TowHookWeapon.ProjSpawned
      (Ability|ApplyGameplayEffectToOwner (Variables|Default|GetGE_Active))
      ;[1 1 1] Save Out the Weapon
      (Utilities|IsValid _returnvalue
        (:Is%20Valid
          (bind _asfort_player_pawn (Utilities|Casting|CastToFortPlayerPawn _returnvalue))
          (Variables|Default|SetFortPawn _asfort_player_pawn)
          ;[1 1 1] If the ability is still running for some reason just kill it.
          (Utilities|Time|SetTimerbyEvent (delegate Custom|ForceEnd) (Variables|Default|GetFailsafeDelay)))))))

(event Custom|ForceEnd
  ;[1 1 1] If the ability is still running for some reason just kill it.
  (Ability|EndAbility))

(event Ability|EventOnEndAbility
  (params (bWasCancelled :type Boolean))
  (|CallParentFunction bWasCancelled)
  (Ability|RemoveGameplayEffectFromOwnerWithGrantedTags (Variables|Default|GetT_Active)))
