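; Tow hook weapon ability that handles tags to block actions like Tac Sprint only while the tethered MME is active.
; On activation it caches the pawn and chains tag-wait tasks for item granted, tethered added and tethered removed, ending each as it fires.
; The first wait also caches the current tow hook weapon.
; Tether removal ends the ability, and ending removes owner gameplay effects that grant the active tag.
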(event Ability|EventOnEndAbility
  (params (bWasCancelled :type Boolean))
  (Ability|RemoveGameplayEffectFromOwnerWithGrantedTags (Variables|Tags|GetT_Active)))

(event Custom|Start%20Tether%20Async%20Tasks
  (bind _fortpawn (Variables|Default|GetFortPawn))
  ;[0.0226 0 1] This GA is used to handle tags to block actions like Tac Sprint specifically while tethered MME is active, not while tethered and in normal movement mod.
  (Utilities|FlowControl|Sequence
    (:then_0
      (bind _asynctaskproxy (Ability|Tasks|WaitGameplayTagAdd (Variables|Tags|GetItem%20Granted%20Tag) _fortpawn)
        (:then
          (Variables|Async|SetWeapon%20Held%20Async%20Task _asynctaskproxy))
        (:Added
          (bind _asb_tow_hook_weapon_athena (Utilities|Casting|CastToB_TowHookWeapon_Athena (Variables|/Script/FortniteGame.FortPawn|GetCurrentWeapon _fortpawn)))
          (Variables|Default|SetWeapon _asb_tow_hook_weapon_athena)
          (GameplayTasks|EndTask (Variables|Async|GetWeapon%20Held%20Async%20Task))
          (bind _asynctaskproxy_2 (Ability|Tasks|WaitGameplayTagAdd (Variables|Tags|GetTethered%20Tag) (Variables|Default|GetFortPawn))
            (:then
              (Variables|Async|SetWait%20Tethered%20Async%20Task _asynctaskproxy_2))
            (:Added
              (GameplayTasks|EndTask (Variables|Async|GetWait%20Tethered%20Async%20Task))
              (bind _asynctaskproxy_3 (Ability|Tasks|WaitGameplayTagRemove (Variables|Tags|GetTethered%20Tag) (Variables|Default|GetFortPawn))
                (:then
                  (Variables|Async|SetTether%20Removed%20Async%20Task _asynctaskproxy_3))
                (:Removed
                  (GameplayTasks|EndTask (Variables|Async|GetTether%20Removed%20Async%20Task))
                  (Ability|EndAbility))))))))))

(event Ability|EventActivateAbility
  ;[0.0226 0 1] This GA is used to handle tags to block actions like Tac Sprint specifically while tethered MME is active, not while tethered and in normal movement mod.
  (|CallParentFunction)
  (bind _asfort_player_pawn (Utilities|Casting|CastToFortPlayerPawn (Ability|GetAvatarActorfromActorInfo)))
  (Variables|Default|SetFortPawn _asfort_player_pawn)
  (CallFunction|StartTetherAsyncTasks))
