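; Passive Athena tow hook weapon ability that watches for conditions that detach a tethered pawn.
; It waits for the item-granted and tethered tags, then starts crouch, jump, tag-removal, vehicle-explosion and player-death watchers.
; Most detach paths clear hit data, end tethered movement and restart the watcher chain, with stored async proxies for cancellation.
; The DBNO-change watcher works around FORT-730543, where the item-granted tag lingers after death.
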(event Ability|EventOnEndAbility
  (params (bWasCancelled :type Boolean))
  ;[1 0.5334 0] End Ability - remove necessary tags, unbind events etc.
  (CallFunction|ClearWeaponHitData)
  (CallFunction|EndTetheredMovementandUnbindEvents))

(event Custom|CrouchWatcher
  ;[0 0.9024 1] Crouch - detach, clean things up, restart singleton async watchers
  (Class|BTowHookWeaponAthena|UpdateHUDKeys (Variables|Default|GetWeapon) true)
  (Utilities|FlowControl|Delay 0.2)
  (bind _asynctaskproxy (Ability|Async|WaitGameplayEventToActor (Variables|Default|GetFortPawn) "(TagName=\"Event.Crouch.Pressed\")")
    (:then
      (Variables|Async|SetCrouch%20Async%20Action _asynctaskproxy))
    (:EventReceived
      (if (not (Variables|/Script/Engine.Character|GetbIsCrouched (Variables|Default|GetFortPawn)))
        (Ability|Async|EndAction (Variables|Async|GetCrouch%20Async%20Action))
        (CallFunction|ResetTetherWatcherStates)))))

(event Custom|DoubleJumpDetachWatcher
  (bind _fortpawn (Variables|Default|GetFortPawn))
  ;[0 1 0.2709] Jump in air - detach, clean things up, restart singleton async watchers
  (EventDispatchers|AssignOnGroundedStateChanged (Actor|GetComponentbyClass (Variables|Default|GetFortPawn) "/Script/TowHookWeaponRuntime.FortPawnComponent_TetheredMovement") (delegate Custom|OnGroundedStateChanged_Event))
  (Pawn|BindEventtoOnPawnJumped _fortpawn (EventDispatchers|CreateEvent :!event Custom|OnJump))
  (EventDispatchers|BindEventtoOnPawnLanded _fortpawn (EventDispatchers|CreateEvent :!event Custom|OnLanded))
  (CallFunction|TetherMMEActiveJumpWatcher))

(event Custom|StartJumpListener
  (bind _fortpawn (Variables|Default|GetFortPawn))
  (bind _returnvalue (Actor|GetComponentbyClass _fortpawn "/Script/TowHookWeaponRuntime.FortPawnComponent_TetheredMovement"))
  (bind _fortpawn_2 (Variables|Default|GetFortPawn))
  ;[0 1 0.2709] Jump in air - detach, clean things up, restart singleton async watchers
  (Ability|Async|EndAction (Variables|Async|GetJump%20Async%20Action))
  (bind _asynctaskproxy (Ability|Async|WaitGameplayEventToActor (Variables|Default|GetFortPawn) "(TagName=\"Event.Jump.Pressed\")" true)
    (:then
      (Variables|Async|SetSecond%20Jump%20Event%20Async%20Action _asynctaskproxy))
    (:EventReceived
      (Class|FortPawnComponentTetheredMovement|EndTetheredMovement _returnvalue)
      (EventDispatchers|UnbindEventfromOnGroundedStateChanged _returnvalue (EventDispatchers|CreateEvent :!event Custom|OnGroundedStateChanged_Event))
      (Pawn|UnbindEventfromOnPawnJumped _fortpawn_2 (EventDispatchers|CreateEvent :!event Custom|OnJump))
      (EventDispatchers|UnbindEventfromOnPawnLanded _fortpawn_2 (EventDispatchers|CreateEvent :!event Custom|OnLanded))
      (Class|BTowHookWeaponAthena|UpdateHUDKeys (Variables|Default|GetWeapon) false)
      (GameplayEffects|RemoveActiveEffectswithGrantedTags (Ability|GetAbilitySystemComponent (Variables|Default|GetFortPawn)) (Variables|Tags|GetTethered%20Any%20State%20Gameplay%20Tags))
      (CallFunction|ResetTetherWatcherStates))))

(event Custom|Start%20Tether%20Async%20Tasks
  (bind _fortpawn (Variables|Default|GetFortPawn))
  ;[0.0226 0 1] Run all our disconnect case watchers
  (bind _asynctaskproxy (Ability|Tasks|WaitGameplayTagAdd (Variables|Tags|GetItem%20Granted%20Tag) _fortpawn)
    (:then
      (Variables|Async|SetWeapon%20Held%20Async%20Task _asynctaskproxy))
    (:Added
      (bind _asb_tow_hook_weapon_athena (Utilities|Casting|CastToB_TowHookWeapon_Athena (Variables|/Script/FortniteGame.FortPawn|GetCurrentWeapon _fortpawn))
        (:then
          (Variables|Default|SetWeapon _asb_tow_hook_weapon_athena)
          (GameplayTasks|EndTask (Variables|Async|GetWeapon%20Held%20Async%20Task))
          (Utilities|FlowControl|Delay 0.2)
          (bind _asynctaskproxy_2 (Ability|Tasks|WaitGameplayTagAdd (Variables|Tags|GetTethered%20Tag%20Any%20State) (Variables|Default|GetFortPawn) true)
            (:then
              (Variables|Async|SetWait%20Tethered%20Async%20Task _asynctaskproxy_2))
            (:Added
              (Utilities|FlowControl|Sequence
                (:then_0
                  (CallFunction|CrouchWatcher))
                (:then_1
                  (CallFunction|DoubleJumpDetachWatcher))
                (:then_2
                  (CallFunction|TetheredTagRemovedWatcher))
                (:then_3
                  (CallFunction|VehicleDeathWatcher))
                (:then_4
                  (CallFunction|PlayerDeathWatcher))))))
        (:CastFailed
          (GameplayTasks|EndTask (Variables|Async|GetWeapon%20Held%20Async%20Task))
          (Utilities|FlowControl|Delay 0.2)
          (CallFunction|StartTetherAsyncTasks))))))

(event Custom|OnGroundedStateChanged_Event
  (params (bGrounded :type Boolean))
  ;[0 1 0.2709] Jump in air - detach, clean things up, restart singleton async watchers
  (if bGrounded
    (Ability|Async|EndAction (Variables|Async|GetJump%20Async%20Action))
    (CallFunction|TetherMMEActiveJumpWatcher)))

(event Custom|OnJump
  (bind _fortpawn (Variables|Default|GetFortPawn))
  (bind _returnvalue (Actor|GetComponentbyClass _fortpawn "/Script/TowHookWeaponRuntime.FortPawnComponent_TetheredMovement"))
  (bind _fortpawn_2 (Variables|Default|GetFortPawn))
  ;[0 1 0.2709] Jump in air - detach, clean things up, restart singleton async watchers
  (bind _asynctaskproxy (Ability|Async|WaitGameplayEventToActor (Variables|Default|GetFortPawn) "(TagName=\"Event.JumpPressed.InAir\")" true)
    (:then
      (Variables|Async|SetJump%20In%20Air%20Async%20Action _asynctaskproxy))
    (:EventReceived
      (Class|FortPawnComponentTetheredMovement|EndTetheredMovement _returnvalue)
      (EventDispatchers|UnbindEventfromOnGroundedStateChanged _returnvalue (EventDispatchers|CreateEvent :!event Custom|OnGroundedStateChanged_Event))
      (Pawn|UnbindEventfromOnPawnJumped _fortpawn_2 (EventDispatchers|CreateEvent :!event Custom|OnJump))
      (EventDispatchers|UnbindEventfromOnPawnLanded _fortpawn_2 (EventDispatchers|CreateEvent :!event Custom|OnLanded))
      (Class|BTowHookWeaponAthena|UpdateHUDKeys (Variables|Default|GetWeapon) false)
      (GameplayEffects|RemoveActiveEffectswithGrantedTags (Ability|GetAbilitySystemComponent (Variables|Default|GetFortPawn)) (Variables|Tags|GetTethered%20Any%20State%20Gameplay%20Tags))
      (CallFunction|ResetTetherWatcherStates))))

(event Ability|EventActivateAbility
  ;[0.0226 0 1] Run all our disconnect case watchers
  (|CallParentFunction)
  (bind _asfort_player_pawn (Utilities|Casting|CastToFortPlayerPawn (Ability|GetAvatarActorfromActorInfo)))
  (Variables|Default|SetFortPawn _asfort_player_pawn)
  (CallFunction|StartTetherAsyncTasks))

(event Custom|TetherMMEActiveJumpWatcher
  ;[0 1 0.2709] Jump in air - detach, clean things up, restart singleton async watchers
  (bind _asynctaskproxy (Ability|Async|WaitGameplayEventToActor (Variables|Default|GetFortPawn) "(TagName=\"Event.Jump.Pressed\")" true)
    (:then
      (Variables|Async|SetJump%20Async%20Action _asynctaskproxy))
    (:EventReceived
      (CallFunction|StartJumpListener))))

(event Custom|OnLanded
  (params (Hit :type Hit%20Result%20Structure%20%28by%20ref%29))
  ;[0 1 0.2709] Jump in air - detach, clean things up, restart singleton async watchers
  (Ability|Async|EndAction (Variables|Async|GetJump%20Async%20Action)))

(event Custom|TetheredTagRemovedWatcher
  (bind _projectilehitdata_projectile_9_ff96757148c1260dff8e428a8db9d153 (Utilities|Struct|BreakTowHookProjectileHit (Variables|/TowHookWeapon/Gameplay/B_TowHookWeapon_Athena.B_TowHookWeapon_Athena_C|GetProjectileHitData_Projectile_9_FF96757148C1260DFF8E428A8DB9D153 (Variables|Default|GetWeapon))))
  ;[0 0.3328 1] Watch for tether tag removed
  (bind _asynctaskproxy (Ability|Tasks|WaitGameplayTagRemove (Variables|Tags|GetTethered%20Tag%20Any%20State) (Variables|Default|GetFortPawn) true)
    (:then
      (Variables|Async|SetTether%20Removed%20Async%20Task _asynctaskproxy))
    (:Removed
      (GameplayTasks|EndTask (Variables|Async|GetWait%20Tethered%20Async%20Task))
      (Ability|Async|EndAction (Variables|Async|GetJump%20Async%20Action))
      (Class|PrjAthenaTowHookWeapon|KillProjectile _projectilehitdata_projectile_9_ff96757148c1260dff8e428a8db9d153)
      (CallFunction|ResetTetherWatcherStates))))

(event Custom|VehicleDeathWatcher
  (bind _weapon (Variables|Default|GetWeapon))
  (bind (_projectilehitdata_projectile_9_ff96757148c1260dff8e428a8db9d153 _projectilehitdata_hitactor_7_424357ae4112f32cb1760fb82d90f298) (Utilities|Struct|BreakTowHookProjectileHit (Variables|/TowHookWeapon/Gameplay/B_TowHookWeapon_Athena.B_TowHookWeapon_Athena_C|GetProjectileHitData_Projectile_9_FF96757148C1260DFF8E428A8DB9D153 _weapon)))
  ;[1 0 0.773] Vehicle Death Watcher
  (Utilities|IsValid _projectilehitdata_hitactor_7_424357ae4112f32cb1760fb82d90f298
    (:Is%20Valid
      (bind _asfort_athena_vehicle (Utilities|Casting|CastToFortAthenaVehicle _projectilehitdata_hitactor_7_424357ae4112f32cb1760fb82d90f298))
      (Variables|Default|SetAs%20Fort%20Athena%20Vehicle _asfort_athena_vehicle)
      (Vehicle|Destruction|BindEventtoOnVehicleExplodeDelegate _asfort_athena_vehicle (delegate Custom|DetachTetherOnVehicleExplode)))))

(event Custom|DetachTetherOnVehicleExplode
  (params (LastDamageInstigator :type Controller%20Object%20Reference) (VehicleActor :type Actor%20Object%20Reference))
  ;[1 0 0.773] Vehicle Death Watcher
  (CallFunction|ResetTetherWatcherStates))

(event Custom|PlayerDeathWatcher
  ;[1 0 0.773] FORT-730543 - granted item tag is not being removed on death. If tow was equipped on death, we immediately fail the Item Granted Async Event watcher on the client, so the client never starts its detach async watchers correctly.
  ; Bind from player dbno state changed. Check if we're really dead and remove the granted item tag manually for now.
  (EventDispatchers|BindEventtoOnDBNOStateChanged (Variables|Default|GetFortPawn) (EventDispatchers|CreateEvent :!event Custom|Custom_PlayerDBNOChanged_Event)))

(event Custom|Custom_PlayerDBNOChanged_Event
  (params (FortPawn :type Fort%20Pawn%20Object%20Reference) (bInIsDBNO :type Boolean))
  ;[1 0 0.773] FORT-730543 - granted item tag is not being removed on death. If tow was equipped on death, we immediately fail the Item Granted Async Event watcher on the client, so the client never starts its detach async watchers correctly.
  ; Bind from player dbno state changed. Check if we're really dead and remove the granted item tag manually for now.
  (if (FortPawn|IsDead FortPawn)
    (Ability|GameplayEffect|RemoveLooseGameplayTags FortPawn (Variables|Default|GetGameplay%20Tags))
    (Ability|EndAbility)))

(fn ClearWeaponHitData
  ;[0.85 1 0] Clear hit actor & location so we can attach to the same vehicle on the next shot if we detach etc.
  (Variables|/TowHookWeapon/Gameplay/B_TowHookWeapon_Athena.B_TowHookWeapon_Athena_C|SetProjectileHitData (Utilities|Struct|MakeTowHookProjectileHit 0) (Variables|Default|GetWeapon)))

(fn EndTetheredMovementAndUnbindEvents
  (bind _as_fort_athena_vehicle (Variables|Default|GetAs%20Fort%20Athena%20Vehicle))
  (bind _fortpawn (Variables|Default|GetFortPawn))
  (Class|FortPawnComponentTetheredMovement|EndTetheredMovement (Actor|GetComponentbyClass (Variables|Default|GetFortPawn) "/Script/TowHookWeaponRuntime.FortPawnComponent_TetheredMovement"))
  (Class|BTowHookWeaponAthena|UpdateHUDKeys (Variables|Default|GetWeapon) false)
  (Ability|RemoveGameplayEffectFromOwnerWithGrantedTags (Variables|Tags|GetMME_Tethered%20Tag))
  (Utilities|IsValid _as_fort_athena_vehicle
    (:Is%20Valid
      (Vehicle|Destruction|UnbindEventfromOnVehicleExplodeDelegate _as_fort_athena_vehicle (EventDispatchers|CreateEvent :!event DetachTetherOnVehicleExplode))
      (Utilities|IsValid _fortpawn
        (:Is%20Valid
          (EventDispatchers|UnbindEventfromOnDBNOStateChanged _fortpawn (EventDispatchers|CreateEvent :!event Custom_PlayerDBNOChanged_Event)))))))

(fn ResetTetherWatcherStates
  (CallFunction|ClearWeaponHitData)
  (CallFunction|EndTetheredMovementandUnbindEvents)
  (CallFunction|StartTetherAsyncTasks))
